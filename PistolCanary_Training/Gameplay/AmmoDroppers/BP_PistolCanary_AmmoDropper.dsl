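; Ammo dropper actor for the PistolCanary training gameplay that spawns loot at its location when a matching event string arrives.
; On BeginPlay, the authority starts listening to the stateful string event on the global GameplayEventRouter.
; Each payload is compared case-insensitively to EventStringToListen; on a match, every WeaponPkg entry is passed to SpawnLootfromTable using the actor's location and up vector.
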(event EventBeginPlay
  (Networking|SwitchHasAuthority
    (:Authority
      (bind (_asynctaskproxy _payload) (GameplayEventRouter|StartListeningToStatefulEvent (GameplayEventRouter|GetGlobalEventRouter) "/Events_Shared/Gameplay/GlobalEvents/Event_RouterEvent_String.Event_RouterEvent_String")
        (:OnEventReceived
          (if (Utilities|String|Equal,CaseInsensitive(String) (Utilities|Struct|BreakEventRouterEventString _payload) (Variables|Default|GetEventStringToListen))
            (for _array_element (Variables|Default|GetWeaponPkg)
              (Class|AthenaFunctionLibrary|SpawnLootfromTable _array_element -1 -1 (Transformation|GetActorLocation) (Transformation|GetActorUpVector) 1 true true))))) ; Listen for Router event to spawn loot.
    )))

(fn ConstructionScript)
